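# Control Rig for the Nosh back bling that procedurally animates the pack's tongue and upper and lower jaws.
# Three Chain Harmonics units run in sequence using only wave motion, starting with a fixed wave on the tongue.
# One seeded random float, remapped and smoothed by accumulated lerps, drives X and Z wave amplitude on both jaw chains, varying their chomping.
# Slime bones are not driven, and debug drawing is enabled on all three units.
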ControlRigBlueprint /BRCosmetics/Accessories/FORT_Backpacks/Backpack_M_MED_Nosh/Meshes/M_MED_Nosh_Pack_CtrlRig
graphs [
 RigVMFunctionLibrary {links [], nodes []}
 RigVMModel {nodes [
   n1 = &RigUnit_BeginExecution()
   /Script/ControlRig.RigUnit_ChainHarmonics(
    ChainRoot=dyn_tongue_base
    DrawWorldOffset=(Rotation=(X=0,Y=0,Z=0,W=1),Translation=(X=0,Y=0,Z=0),Scale3D=(X=1,Y=1,Z=1))
    Pendulum=(bEnabled=False,PendulumStiffness=2,PendulumGravity=(X=0,Y=0,Z=0),PendulumBlend=0.75,PendulumDrag=0.98,PendulumMinimum=0,PendulumMaximum=0.1,PendulumEase=Linear,UnwindAxis=(X=0,Y=1,Z=0),UnwindMinimum=0.2,UnwindMaximum=0.05)
    Reach=(bEnabled=False,ReachTarget=(X=0,Y=0,Z=0),ReachAxis=(X=1,Y=0,Z=0),ReachMinimum=0,ReachMaximum=0,ReachEase=SineEaseInOut)
    Speed=(X=1,Y=1.2,Z=1)
    Wave=(bEnabled=True,WaveFrequency=(X=0,Y=1,Z=1),WaveAmplitude=(X=0,Y=0.5,Z=0),WaveOffset=(X=0,Y=2,Z=1),WaveNoise=(X=0,Y=4,Z=0),WaveMinimum=1,WaveMaximum=3,WaveEase=SineEaseInOut)
    WaveCurve=(EditorCurveData=(Keys=((),(Time=1,Value=1)),PreInfinityExtrap=RCCE_Constant,PostInfinityExtrap=RCCE_Constant,DefaultValue=340282346638528859811704183484516925440))
    bDrawDebug=True
    ExecutePin=@n1.ExecutePin
   ) {is=struct RigVMUnitNode}
   n2 = /Script/RigVM.RigVMFunction_RandomFloat(Duration=3, Maximum=1, Minimum=0, Seed=217) {is=struct RigVMUnitNode, template="Random::Execute(in Seed,in Minimum,in Maximum,in Duration,out Result)"}
   > &RigUnit_ChainHarmonics(
    ChainRoot=dyn_jaw_02
    DrawWorldOffset=(Rotation=(X=0,Y=0,Z=0,W=1),Translation=(X=0,Y=0,Z=0),Scale3D=(X=1,Y=1,Z=1))
    Pendulum=(bEnabled=False,PendulumStiffness=2,PendulumGravity=(X=0,Y=0,Z=0),PendulumBlend=0.75,PendulumDrag=0.98,PendulumMinimum=0,PendulumMaximum=0.1,PendulumEase=Linear,UnwindAxis=(X=0,Y=1,Z=0),UnwindMinimum=0.2,UnwindMaximum=0.05)
    Reach=(bEnabled=False,ReachTarget=(X=0,Y=0,Z=0),ReachAxis=(X=1,Y=0,Z=0),ReachMinimum=0,ReachMaximum=0,ReachEase=SineEaseInOut)
    Speed=(X=2,Y=0,Z=2)
    Wave=(bEnabled=True,WaveFrequency=(X=0,Y=1000,Z=0),WaveAmplitude=(X=0.25,Y=0,Z=0.25),WaveOffset=(X=0,Y=1,Z=0),WaveNoise=(X=2.5,Y=0,Z=2.5),WaveMinimum=1,WaveMaximum=0,WaveEase=SineEaseInOut)
    WaveCurve=(EditorCurveData=(Keys=((Time=0.8813,Value=0.1615),(Time=0.9663,Value=1.0036)),PreInfinityExtrap=RCCE_Constant,PostInfinityExtrap=RCCE_Constant,DefaultValue=340282346638528859811704183484516925440))
    bDrawDebug=True
    Wave.WaveAmplitude.Z=/Script/RigVM.RigVMFunction_AccumulateFloatLerp(
     Blend=16
     InitialValue=1
     bIntegrateDeltaTime=true
     TargetValue=&RigVMFunction_MathFloatRemap(SourceMaximum=1, SourceMinimum=0, TargetMaximum=0.3, TargetMinimum=0, bClamp=False, Value=@n2.Result).Result
    ) {is=struct RigVMUnitNode, template="AccumulateLerp::Execute(in TargetValue,in InitialValue,in Blend,in bIntegrateDeltaTime,out Result)"}.Result
    Wave.WaveAmplitude.X=&RigVMFunction_AccumulateFloatLerp(
     Blend=16
     InitialValue=1
     bIntegrateDeltaTime=true
     TargetValue=&RigVMFunction_MathFloatRemap(SourceMaximum=1, SourceMinimum=0, TargetMaximum=0.3, TargetMinimum=0, bClamp=false, Value=@n2.Result).Result
    ) {template="AccumulateLerp::Execute(in TargetValue,in InitialValue,in Blend,in bIntegrateDeltaTime,out Result)"}.Result
   )
   > &RigUnit_ChainHarmonics(
    ChainRoot=dyn_jaw_upper_02
    DrawWorldOffset=(Rotation=(X=0,Y=0,Z=0,W=1),Translation=(X=0,Y=0,Z=0),Scale3D=(X=1,Y=1,Z=1))
    Pendulum=(bEnabled=False,PendulumStiffness=2,PendulumGravity=(X=0,Y=0,Z=0),PendulumBlend=0.75,PendulumDrag=0.98,PendulumMinimum=0,PendulumMaximum=0.1,PendulumEase=Linear,UnwindAxis=(X=0,Y=1,Z=0),UnwindMinimum=0.2,UnwindMaximum=0.05)
    Reach=(bEnabled=False,ReachTarget=(X=0,Y=0,Z=0),ReachAxis=(X=1,Y=0,Z=0),ReachMinimum=0,ReachMaximum=0,ReachEase=SineEaseInOut)
    Speed=(X=1,Y=0,Z=1)
    Wave=(bEnabled=True,WaveFrequency=(X=0,Y=1000,Z=0),WaveAmplitude=(X=0.25,Y=0,Z=0.25),WaveOffset=(X=0,Y=1,Z=0),WaveNoise=(X=2,Y=0,Z=2),WaveMinimum=1,WaveMaximum=0,WaveEase=SineEaseInOut)
    WaveCurve=(EditorCurveData=(Keys=((Time=0.2257,Value=0.2271),(Time=1,Value=1)),PreInfinityExtrap=RCCE_Constant,PostInfinityExtrap=RCCE_Constant,DefaultValue=340282346638528859811704183484516925440))
    bDrawDebug=True
    Wave.WaveAmplitude.X=&RigVMFunction_AccumulateFloatLerp(
     Blend=14
     InitialValue=1
     bIntegrateDeltaTime=true
     TargetValue=&RigVMFunction_MathFloatRemap(SourceMaximum=1, SourceMinimum=0, TargetMaximum=0.3, TargetMinimum=0, bClamp=False, Value=@n2.Result).Result
    ) {template="AccumulateLerp::Execute(in TargetValue,in InitialValue,in Blend,in bIntegrateDeltaTime,out Result)"}.Result
    Wave.WaveAmplitude.Z=&RigVMFunction_AccumulateFloatLerp(
     Blend=14
     InitialValue=1
     bIntegrateDeltaTime=true
     TargetValue=&RigVMFunction_MathFloatRemap(SourceMaximum=1, SourceMinimum=0, TargetMaximum=0.3, TargetMinimum=0, bClamp=False, Value=@n2.Result).Result
    ) {template="AccumulateLerp::Execute(in TargetValue,in InitialValue,in Blend,in bIntegrateDeltaTime,out Result)"}.Result
   )
  ]}
]
hierarchy [
 root {loc=<0 0 0>, rot=<0 0 0 1>}
 pack {parent=root, loc=<0 -0.7316 0>, rot=<0 0 0 1>}
 dyn_master {parent=pack, loc=<-0.126 -9.1738 -11.136>, rot=<-0.5 -0.5 0.5 -0.5>}
 dyn_slime_01 {parent=dyn_master, loc=<5.1216 3.7678 -1.8384>, rot=<0 0 0 1>}
 dyn_slime_02 {parent=dyn_master, loc=<4.459 -5.2877 6.4431>, rot=<0 0 0 1>}
 dyn_slime_03 {parent=dyn_master, loc=<2.1099 -7.0884 -2.2965>, rot=<0 0 0 1>}
 dyn_jaw_02 {parent=pack, loc=<0.3169 -8.5313 -3.9644>, rot=<0 0 0.7071 -0.7071>}
 dyn_jaw_01 {parent=dyn_jaw_02, loc=<2.6006 0 0>, rot=<0 0 0 -1>}
 dyn_jaw {parent=dyn_jaw_01, loc=<2.8031 -0.4139 0.1889>, rot=<0 -0.1736 0 -0.9848>}
 dyn_jaw_upper_02 {parent=pack, loc=<0.3169 -8.5109 -2.5842>, rot=<0 0 0.7071 -0.7071>}
 dyn_jaw_upper_01 {parent=dyn_jaw_upper_02, loc=<2.6211 0 0>, rot=<0 0 0 -1>}
 dyn_jaw_upper {parent=dyn_jaw_upper_01, loc=<2.6418 -0.4642 0.0748>, rot=<0 0.2588 0 -0.9659>}
 dyn_tongue_base {parent=pack, loc=<0.1835 -11.8456 -7.4931>, rot=<0.2534 0.2534 0.6601 -0.6601>}
 dyn_tongue_parent {parent=dyn_tongue_base, loc=<0 0 0>, rot=<0 0 0 -1>}
 dyn_tongue_01 {parent=dyn_tongue_parent, loc=<4.3588 0.0084 -1.5121>, rot=<0 -0.454 0 -0.891>}
 dyn_tongue_02 {parent=dyn_tongue_01, loc=<2.6233 -0.0659 0.0744>, rot=<0 -0.1994 0 -0.9799>}
]
