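# Physics asset for the tongue on the Nosh backpack cosmetic, defining bodies and joint limits for its dynamic bones.
# A small box at the tongue base anchors a chain of three identical flat boxes running from parent through 01 to 02.
# Constraints with swing and twist limits of only a few degrees keep the tongue fairly stiff.
# A capsule on the pack bone has no constraint, apparently serving as a collision volume for the backpack body.
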PhysicsAsset /BRCosmetics/Accessories/FORT_Backpacks/Backpack_M_MED_Nosh/Meshes/M_MED_Nosh_Pack_Tongue_Physics
bodies {
 dyn_tongue_01 [{shapeType=Box, extentX=2.6726, extentY=8.1561, extentZ=1.0342, center=<1.0798 0 0>}]
 dyn_tongue_02 [{shapeType=Box, extentX=2.6726, extentY=8.1561, extentZ=1.0342, center=<1.0798 0 0>}]
 dyn_tongue_base [{shapeType=Box, extentX=1, extentY=0.5, extentZ=1, rotation=<rot 1.5551 -0.0242 0.167>}]
 dyn_tongue_parent [{shapeType=Box, extentX=2.6726, extentY=8.1561, extentZ=1.0342, center=<1.0798 0 0>}]
 pack [{shapeType=Capsule, length=3.125, radius=4.5, center=<0 -16.3748 -10.9278>, rotation=<rot 90 0 0>}]
}
constraints [
 {bone1Name=dyn_tongue_01, bone2Name=dyn_tongue_parent, swing1LimitDegrees=3, swing2LimitDegrees=2, twistLimitDegrees=3}
 {bone1Name=dyn_tongue_02, bone2Name=dyn_tongue_01, swing1LimitDegrees=3, swing2LimitDegrees=4, twistLimitDegrees=3}
 {bone1Name=dyn_tongue_parent, bone2Name=dyn_tongue_base, swing1LimitDegrees=3, swing2LimitDegrees=2, twistLimitDegrees=3}
]
