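# Physics asset for the Nosh backpack cosmetic's slime pack mesh, defining collision bodies and joints for three dynamic slime bones.
# A larger capsule covers the pack bone, and a small offset capsule covers each slime bone.
# Each slime bone is constrained directly to the pack rather than chained, with a tight twist limit and its own modest swing cone, so the three move by differing amounts.
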PhysicsAsset /BRCosmetics/Accessories/FORT_Backpacks/Backpack_M_MED_Nosh/Meshes/M_MED_Nosh_Slime_Pack_Physics
bodies {
 dyn_slime_01 [{shapeType=Capsule, length=1.7089, radius=1.1187, center=<1.2725 0 0>, rotation=<rot 90 0 0>}]
 dyn_slime_02 [{shapeType=Capsule, length=1.2235, radius=1.3787, center=<1.6166 0 0>, rotation=<rot 90 0 0>}]
 dyn_slime_03 [{shapeType=Capsule, length=0.9751, radius=0.9658, center=<1.0339 0 0>, rotation=<rot 90 0 0>}]
 pack [{shapeType=Capsule, length=5, radius=3.125}]
}
constraints [
 {bone1Name=dyn_slime_01, bone2Name=pack, swing1LimitDegrees=20, swing2LimitDegrees=20, twistLimitDegrees=5}
 {bone1Name=dyn_slime_02, bone2Name=pack, swing1LimitDegrees=25, swing2LimitDegrees=25, twistLimitDegrees=5}
 {bone1Name=dyn_slime_03, bone2Name=pack, swing1LimitDegrees=10, swing2LimitDegrees=10, twistLimitDegrees=5}
]
