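# Physics asset for the upper-jaw mesh of the Nosh backpack cosmetic, defining simulation bodies for its dynamic jaw bones.
# Thin box bodies on dyn_jaw and dyn_jaw_upper each constrain to a small capsule on the pack bone.
# The dyn_jaw_upper constraint locks swing1 and twist, leaving swing2 as a hinge-like axis, while dyn_jaw has no locked angular axis.
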PhysicsAsset /BRCosmetics/Accessories/FORT_Backpacks/Backpack_M_MED_Nosh/Meshes/M_MED_Nosh_UpperJaw_Pack_Physics
bodies {
 dyn_jaw [{shapeType=Box, extentX=12.3983, extentY=14.3935, extentZ=1.1457, center=<6.5218 0.0001 0>}]
 dyn_jaw_upper [{shapeType=Box, extentX=12.3983, extentY=14.3935, extentZ=1.1457, center=<5.906 0 0>}]
 pack [{shapeType=Capsule, length=5, radius=3.125}]
}
constraints [
 {bone1Name=dyn_jaw, bone2Name=pack, swing1LimitDegrees=3, swing2LimitDegrees=15, twistLimitDegrees=2}
 {
  bone1Name=dyn_jaw_upper
  bone2Name=pack
  swing1LimitDegrees=1
  swing1Motion=Locked
  swing2LimitDegrees=15
  twistLimitDegrees=2
  twistMotion=Locked
 }
]
